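# Animation Blueprint for the SizeTribe glider, animating deploy, flight and closing, with a separate front-end path for cosmetic preview.
# Open/closed state machines and aim blendspaces form the base, layered with additive lean, turn, jitter, facial and blink animation.
# Cached poses then add wrist-to-wing two-bone IK outside preview, plus leg and tail bone adjustments.
# The EventGraph updates jitter rate, tail spin and blink timing, while notifies hide or reveal Glider_Base.
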AnimBlueprint /BRCosmetics/Animation/Game/MainPlayer/Skydive/ParaGlide/SizeTribe/SizeTribe_Glider_AnimBP
target_skeleton=/BRCosmetics/Gadgets/Assets/VinderTech_GliderChute/Glider_SizeTribe/Meshes/SizeTribe_Glider_Skeleton
graphs [
 AnimGraph {nodes [
   Output Pose(Result=Slot 'DefaultSlot'(Source=Use cached pose 'Ik Setup'())) {layerGroup=None}
   AnimGraphNode_StateMachine_1 = Locomotion SM() {is=title AnimGraphNode_StateMachine}
   AnimGraphNode_StateMachine_2 = Lean Transitions() {is=title AnimGraphNode_StateMachine}
   n1 = Apply Additive!(Base=@AnimGraphNode_StateMachine_1, Additive=@AnimGraphNode_StateMachine_2, AlphaCurveName:Name=ParaGlide_Lean) -> (Pose) {alphaInputType=Curve, alphaBoolBlend {blendOption=Linear}, alphaScaleBiasClamp {bias=1, scale=-1}}
   n2 = Blendspace Player 'BS_SizeTribe_Turn_GLIDER'(X=0, Y=0) {blendSpace=&/BS_SizeTribe_Turn_GLIDER}
   AnimGraphNode_StateMachine_0 = FE SM() {is=title AnimGraphNode_StateMachine}
   n3 = Get OwnerGlider() -> (OwnerGlider:Fort Player Parachute Object Reference)
   n4 = Apply Additive!(
    Base=Blend Poses by bool(
     BlendPose_0=@AnimGraphNode_StateMachine_0
     BlendPose_1=Apply Additive(Base=@n1, Additive=@n2) {alphaBoolBlend {blendOption=Linear}}
     bActiveValue=Get IsCosmeticPreview!(self:Fort Player Parachute Object Reference=@n3) -> (IsCosmeticPreview:bool)
    ) {blendType=Linear}
    Additive="Sequence Player (additive) 'SizeTribe_Jitter_Anim_GLIDER'"!(PlayRate:float=Get AccelRate() -> (AccelRate:double)) -> (Pose) {sequence=&/SizeTribe_Jitter_Anim_GLIDER}
    Alpha:float
   ) -> (Pose) {alphaBoolBlend {blendOption=Linear}}
   n5 = "Sequence Player (additive) 'SizeTribe_Facial_Anim_GLIDER'"() {sequence=&/SizeTribe_Facial_Anim_GLIDER}
   AnimGraphNode_StateMachine_3 = Blink SM() {is=title AnimGraphNode_StateMachine}
   Save cached pose 'Locomotion Animation'(Pose=Apply Additive(
     Base=Apply Additive!(Base=@n4, Additive=@n5, AlphaCurveName:Name=blink_curve) -> (Pose) {alphaInputType=Curve, alphaBoolBlend {blendOption=Linear}, alphaScaleBiasClamp {bias=1, scale=-1}}
     Additive=@AnimGraphNode_StateMachine_3
    ) {alphaBoolBlend {blendOption=Linear}}) {cacheName=Locomotion Animation}
   n6 = Local To Component(LocalPose=Use cached pose 'IKLegsCache'())
   n7 = "Transform (Modify) Bone - Bone: joint_tails_end"!(
    ComponentPose="Transform (Modify) Bone - Bone: joint_tails_end"!(
     ComponentPose="Transform (Modify) Bone - Bone: joint_tails"!(ComponentPose=@n6, bAlphaBoolEnabled:bool=True, Translation="0,0,0", Rotation="0,0,0", Scale="1,1,1") -> (Pose) {alphaInputType=Bool, rotationMode=BMM_Replace, translationMode=BMM_Additive, boneToModify {boneName=joint_tails}}
     bAlphaBoolEnabled:bool=True
     Translation="0,0,0"
     Rotation="0,0,0"
     Scale="1,1,1"
    ) -> (Pose) {alphaInputType=Bool, rotationMode=BMM_Replace, boneToModify {boneName=joint_tails_end}}
    AlphaCurveName:Name=TailModified
    Translation="0,10,10"
    Rotation="0,0,0"
    Scale="1,1,1"
   ) -> (Pose) {
    alphaInputType=Curve
    translationMode=BMM_Additive
    alphaScaleBiasClamp {bias=1, scale=-1}
    boneToModify {boneName=joint_tails_end}
    translation {y=10, z=10}
   }
   n8 = Get OwnerGlider()
   n9 = "Transform (Modify) Bone - Bone: joint_tails"(
    ComponentPose=Local To Component(LocalPose=Use cached pose 'IKLegsCache'())
    bAlphaBoolEnabled=True
    Translation="0,0,0"
    Rotation="0,0,0"
    Scale="1,1,1"
   ) {
    alphaInputType=Bool
    rotationMode=BMM_Replace
    translationMode=BMM_Additive
    translationSpace=BCS_BoneSpace
    boneToModify {boneName=joint_tails}
   }
   n10 = "Transform (Modify) Bone - Bone: joint_tails_end"(ComponentPose=@n9, bAlphaBoolEnabled=True, Translation="0,5,5", Rotation="0,0,0", Scale="1,1,1") {alphaInputType=Bool, rotationMode=BMM_Replace, translationMode=BMM_Additive, boneToModify {boneName=joint_tails_end}, translation {y=5, z=5}}
   Save cached pose 'Ik Setup'(Pose=Blend Poses by bool(
     BlendPose_0=Component To Local(ComponentPose=@n7)
     BlendPose_1=Component To Local(ComponentPose=@n10)
     bActiveValue=Get IsCosmeticPreview(self=@n8)
    )) {cacheName=Ik Setup}
   n11 = "Transform (Modify) Bone - Bone: joint_boots_leg_l1"(
    ComponentPose=Local To Component(LocalPose=Use cached pose 'Locomotion Animation'())
    Alpha=1
    Translation="8,0,0"
    Rotation="0,0,0"
    Scale="1,1,1"
   ) {translationMode=BMM_Additive, translationSpace=BCS_ParentBoneSpace, boneToModify {boneName=joint_boots_leg_l1}, translation {x=8}}
   n12 = Get OwnerGlider()
   n13 = Two Bone IK - Bone: joint_wrist_l!(ComponentPose:Component Space Pose Link Structure=Local To Component(LocalPose=Use cached pose 'Locomotion Animation'()), Alpha:float=1) -> (Pose:Component Space Pose Link Structure) {
    is=title AnimGraphNode_TwoBoneIK
    bEnableDebugDraw=true
    bAllowStretching=true
    bAllowTwist=false
    bTakeRotationFromEffectorSpace=true
    effectorLocationSpace=BCS_BoneSpace
    effectorTarget {boneReference {boneName=Wing_l}}
    iKBone {boneName=joint_wrist_l}
    jointTarget {boneReference {boneName=joint_elbow_l_IK}}
   }
   n14 = Two Bone IK - Bone: joint_wrist_r!(ComponentPose:Component Space Pose Link Structure=@n13, Alpha:float=1) -> (Pose:Component Space Pose Link Structure) {
    is=title AnimGraphNode_TwoBoneIK
    bAllowStretching=true
    bAllowTwist=false
    bTakeRotationFromEffectorSpace=true
    effectorLocationSpace=BCS_BoneSpace
    effectorTarget {boneReference {boneName=Wing_r}}
    iKBone {boneName=joint_wrist_r}
   }
   n15 = "Transform (Modify) Bone - Bone: joint_boots_leg_l1"(ComponentPose=@n14, Alpha=1, Translation="8,0,0", Rotation="0,0,0", Scale="1,1,1") {translationMode=BMM_Additive, translationSpace=BCS_ParentBoneSpace, boneToModify {boneName=joint_boots_leg_l1}, translation {x=8}}
   Save cached pose 'IKLegsCache'(Pose=Blend Poses by bool(
     BlendPose_0=Component To Local(ComponentPose=@n11)
     BlendPose_1=Component To Local(ComponentPose=@n15)
     bActiveValue=Get IsCosmeticPreview(self=@n12)
    )) {cacheName=IKLegsCache}
  ]}
 FE SM {
  entry=Closed
  path=AnimGraph.AnimGraphNode_StateMachine_0.FE SM
  ids=id
  nodes [
   AnimStateNode_0 = Idle_Open() {is=title AnimStateNode}
   AnimStateNode_1 = Open_Into() {is=title AnimStateNode}
   AnimStateNode_2 = Closed() {is=title AnimStateNode}
   AnimStateTransitionNode_0 = AnimStateTransitionNode() {bAutomaticRuleBasedOnSequencePlayerInState=true, blendMode=Linear, crossfadeDuration=0}
   AnimStateTransitionNode_1 = AnimStateTransitionNode() {crossfadeDuration=0}
   AnimStateTransitionNode_2 = AnimStateTransitionNode() {crossfadeDuration=0}
   AnimStateTransitionNode_3 = AnimStateTransitionNode() {crossfadeDuration=0, priorityOrder=2}
  ]
  links [
   {
    dst=Open_Into
    rule=AnimGraph.AnimGraphNode_StateMachine_0.FE SM.AnimStateTransitionNode_1.Transition
    src=Closed
    via=AnimStateTransitionNode_1
   }
   {
    dst=Closed
    rule=AnimGraph.AnimGraphNode_StateMachine_0.FE SM.AnimStateTransitionNode_2.Transition
    src=Idle_Open
    via=AnimStateTransitionNode_2
   }
   {
    dst=Closed
    rule=AnimGraph.AnimGraphNode_StateMachine_0.FE SM.AnimStateTransitionNode_3.Transition
    src=Open_Into
    via=AnimStateTransitionNode_3
   }
   {
    dst=Idle_Open
    rule=AnimGraph.AnimGraphNode_StateMachine_0.FE SM.AnimStateTransitionNode_0.Transition
    src=Open_Into
    via=AnimStateTransitionNode_0
   }
  ]
 }
 Idle_Open {
  path=AnimGraph.AnimGraphNode_StateMachine_0.FE SM.AnimStateNode_0.Idle_Open
  nodes [Output Animation Pose(Result=Sequence Player 'SizeTribe_Glide_FrontEnd_Anim_GLIDER'() {sequence=&/SizeTribe_Glide_FrontEnd_Anim_GLIDER})]
 }
 Open_Into {
  path=AnimGraph.AnimGraphNode_StateMachine_0.FE SM.AnimStateNode_1.Open_Into
  nodes [
   n1 = Sequence Player 'SizeTribe_Into_FrontEnd_Anim_GLIDER'() {bLoopAnimation=false, sequence=&/SizeTribe_Into_FrontEnd_Anim_GLIDER}
   Output Animation Pose(Result=@n1)
  ]
 }
 Closed {
  path=AnimGraph.AnimGraphNode_StateMachine_0.FE SM.AnimStateNode_2.Closed
  nodes [
   n1 = Get OwnerGlider()
   Output Animation Pose(Result=Blend Poses by bool(
     BlendPose_0=Sequence Player 'SizeTribe_Closed_GLIDER'() {sequence=&/SizeTribe_Closed_GLIDER}
     BlendPose_1=Sequence Player 'SizeTribe_Closed_GLIDER'() {sequence=&SizeTribe_Closed_GLIDER}
     bActiveValue=Get IsCosmeticPreview(self=@n1)
    ) {blendType=Linear})
  ]
 }
 Transition {
  path=AnimGraph.AnimGraphNode_StateMachine_0.FE SM.AnimStateTransitionNode_0.Transition
  nodes [
   Result()
   "Equal (Float)"!(A:double="Get Relevant Anim Time Remaining (Open_Into)"!() -> (ReturnValue:float) {is=title K2Node_AnimGetter}, B:double=0) -> (ReturnValue:bool) {is=title K2Node_CallFunction}
  ]
 }
 Transition {
  path=AnimGraph.AnimGraphNode_StateMachine_0.FE SM.AnimStateTransitionNode_1.Transition
  nodes [
   n1 = Get SkydivingState() -> (SkydivingState:Fort Anim Input Skydiving Structure)
   Result(bCanEnterTransition=OR Boolean(
     A=Break Fort Anim Input Skydiving!(FortAnimInput_Skydiving:"Fort Anim Input Skydiving Structure (by ref)"=@n1) -> (
      ParaGilderRootOffset:Vector=~
      ParaGliderRootRotationOverride:Rotator=~
      bUseParaGlideRootModifier:bool=~
      bIsSkydivingFromLaunchPad:bool=~
      bIsSkydivingFromBus:bool=~
      bIsInVortex:bool=~
      bIsDiving:bool=~
      bIsParachuteOpen:bool=~
      bIsSkydiving:bool=~
      bIsParachuteLeaning:bool=~
      bIsSkydiveLeaning:bool=~
      bIsLeaning:bool=~
      bIsSkydiveDiveMode:bool=~
      bPlaySkydiveDrift:bool=~
      bSkydiveDriftDelayActive:bool=~
      bSkydiveDriftAnimAllowed:bool=~
      bIsGliderRight:bool=~
      bIsGliderCenter:bool=~
      bIsGliderLeft:bool=~
      bIsGliderForward:bool=~
      bIsGliderBack:bool=~
      bHasDeployedAndIsStandingUpright:bool=~
      LocalAccelForward:float=~
      LocalAccelRight:float=~
      LocalVelocityRight:float=~
      SkydiveAimPitch:float=~
      SkydiveAimYaw:float=~
      DeployChuteAnimRate:float=~
      SkydiveDriftAnimRate:float=~
      SkydiveDriftAnimRateCurrent:float=~
      SkydiveFidgetAnimRate:float=~
      SkydiveFidgetAnimRateCurrent:float=~
      SkydiveAdditiveAlpha:float=~
      SkydiveDriftDelay:float=~
      SkydiveDriftAnim:int
      SkydiveDriftAnimMax:int
      LaunchpadAnim:int
      LocalAccelDir:ESkydivingDirection Enum
      DirectionLast:ESkydivingDirection Enum
      PoseAdditive:Anim Sequence Object Reference
     ) {is=title K2Node_BreakStruct}.bIsParachuteOpen
     B=Get bForceOpen() -> (bForceOpen:bool)
    ))
  ]
 }
 Transition {
  path=AnimGraph.AnimGraphNode_StateMachine_0.FE SM.AnimStateTransitionNode_2.Transition
  nodes [
   n1 = Get bForceOpen()
   n2 = Break Fort Anim Input Skydiving(FortAnimInput_Skydiving=Get SkydivingState()) -> (
    ParaGilderRootOffset=~
    ParaGliderRootRotationOverride=~
    bUseParaGlideRootModifier=~
    bIsSkydivingFromLaunchPad=~
    bIsSkydivingFromBus=~
    bIsInVortex=~
    bIsDiving=~
    bIsParachuteOpen=~
    bIsSkydiving=~
    bIsParachuteLeaning=~
    bIsSkydiveLeaning=~
    bIsLeaning=~
    bIsSkydiveDiveMode=~
    bPlaySkydiveDrift=~
    bSkydiveDriftDelayActive=~
    bSkydiveDriftAnimAllowed=~
    bIsGliderRight=~
    bIsGliderCenter=~
    bIsGliderLeft=~
    bIsGliderForward=~
    bIsGliderBack=~
    bHasDeployedAndIsStandingUpright=~
    LocalAccelForward=~
    LocalAccelRight=~
    LocalVelocityRight=~
    SkydiveAimPitch=~
    SkydiveAimYaw=~
    DeployChuteAnimRate=~
    SkydiveDriftAnimRate=~
    SkydiveDriftAnimRateCurrent=~
    SkydiveFidgetAnimRate=~
    SkydiveFidgetAnimRateCurrent=~
    SkydiveAdditiveAlpha=~
    SkydiveDriftDelay=~
   )
   Result(bCanEnterTransition=AND Boolean(A=NOT Boolean(A=@n2.bIsParachuteOpen), B=NOT Boolean(A=@n1)))
  ]
 }
 Transition {
  path=AnimGraph.AnimGraphNode_StateMachine_0.FE SM.AnimStateTransitionNode_3.Transition
  nodes [
   n1 = Get bForceOpen()
   n2 = Break Fort Anim Input Skydiving(FortAnimInput_Skydiving=Get SkydivingState()) -> (
    ParaGilderRootOffset=~
    ParaGliderRootRotationOverride=~
    bUseParaGlideRootModifier=~
    bIsSkydivingFromLaunchPad=~
    bIsSkydivingFromBus=~
    bIsInVortex=~
    bIsDiving=~
    bIsParachuteOpen=~
    bIsSkydiving=~
    bIsParachuteLeaning=~
    bIsSkydiveLeaning=~
    bIsLeaning=~
    bIsSkydiveDiveMode=~
    bPlaySkydiveDrift=~
    bSkydiveDriftDelayActive=~
    bSkydiveDriftAnimAllowed=~
    bIsGliderRight=~
    bIsGliderCenter=~
    bIsGliderLeft=~
    bIsGliderForward=~
    bIsGliderBack=~
    bHasDeployedAndIsStandingUpright=~
    LocalAccelForward=~
    LocalAccelRight=~
    LocalVelocityRight=~
    SkydiveAimPitch=~
    SkydiveAimYaw=~
    DeployChuteAnimRate=~
    SkydiveDriftAnimRate=~
    SkydiveDriftAnimRateCurrent=~
    SkydiveFidgetAnimRate=~
    SkydiveFidgetAnimRateCurrent=~
    SkydiveAdditiveAlpha=~
    SkydiveDriftDelay=~
   )
   Result(bCanEnterTransition=AND Boolean(A=NOT Boolean(A=@n2.bIsParachuteOpen), B=NOT Boolean(A=@n1)))
  ]
 }
 Locomotion SM {
  entry=Closed
  path=AnimGraph.AnimGraphNode_StateMachine_1.Locomotion SM
  ids=id
  nodes [
   AnimStateNode_0 = Idle_Open()
   AnimStateNode_1 = Open_Into()
   AnimStateNode_2 = Closed()
   AnimStateTransitionNode_0 = AnimStateTransitionNode() {bAutomaticRuleBasedOnSequencePlayerInState=true, blendMode=Linear, crossfadeDuration=0}
   AnimStateTransitionNode_1 = AnimStateTransitionNode() {crossfadeDuration=0.1}
   AnimStateTransitionNode_2 = AnimStateTransitionNode() {crossfadeDuration=0}
   AnimStateTransitionNode_3 = AnimStateTransitionNode() {crossfadeDuration=0, priorityOrder=2}
  ]
  links [
   {
    dst=Open_Into
    rule=AnimGraph.AnimGraphNode_StateMachine_1.Locomotion SM.AnimStateTransitionNode_1.Transition
    src=Closed
    via=AnimStateTransitionNode_1
   }
   {
    dst=Closed
    rule=AnimGraph.AnimGraphNode_StateMachine_1.Locomotion SM.AnimStateTransitionNode_2.Transition
    src=Idle_Open
    via=AnimStateTransitionNode_2
   }
   {
    dst=Closed
    rule=AnimGraph.AnimGraphNode_StateMachine_1.Locomotion SM.AnimStateTransitionNode_3.Transition
    src=Open_Into
    via=AnimStateTransitionNode_3
   }
   {
    dst=Idle_Open
    rule=AnimGraph.AnimGraphNode_StateMachine_1.Locomotion SM.AnimStateTransitionNode_0.Transition
    src=Open_Into
    via=AnimStateTransitionNode_0
   }
  ]
 }
 Idle_Open {
  path=AnimGraph.AnimGraphNode_StateMachine_1.Locomotion SM.AnimStateNode_0.Idle_Open
  nodes [
   n1 = Break Fort Anim Input Skydiving(FortAnimInput_Skydiving=Get SkydivingState()) -> (
    ParaGilderRootOffset=~
    ParaGliderRootRotationOverride=~
    bUseParaGlideRootModifier=~
    bIsSkydivingFromLaunchPad=~
    bIsSkydivingFromBus=~
    bIsInVortex=~
    bIsDiving=~
    bIsParachuteOpen=~
    bIsSkydiving=~
    bIsParachuteLeaning=~
    bIsSkydiveLeaning=~
    bIsLeaning=~
    bIsSkydiveDiveMode=~
    bPlaySkydiveDrift=~
    bSkydiveDriftDelayActive=~
    bSkydiveDriftAnimAllowed=~
    bIsGliderRight=~
    bIsGliderCenter=~
    bIsGliderLeft=~
    bIsGliderForward=~
    bIsGliderBack=~
    bHasDeployedAndIsStandingUpright=~
    LocalAccelForward=~
    LocalAccelRight=~
    LocalVelocityRight=~
    SkydiveAimPitch=~
    SkydiveAimYaw=~
    DeployChuteAnimRate=~
    SkydiveDriftAnimRate=~
    SkydiveDriftAnimRateCurrent=~
    SkydiveFidgetAnimRate=~
    SkydiveFidgetAnimRateCurrent=~
    SkydiveAdditiveAlpha=~
    SkydiveDriftDelay=~
   )
   n2 = Blendspace Player 'BS_SizeTribe_GLIDER'(X=@n1.SkydiveAimYaw, Y=@n1.SkydiveAimPitch) {blendSpace=&/BS_SizeTribe_GLIDER, groupName=Glide, method=SyncGroup}
   n3 = Get OwnerGlider()
   Output Animation Pose(Result=Blend Poses by bool(
     BlendPose_0=Sequence Player 'SizeTribe_Glide_FrontEnd_Anim_GLIDER'() {sequence=&SizeTribe_Glide_FrontEnd_Anim_GLIDER}
     BlendPose_1=@n2
     bActiveValue=Get IsCosmeticPreview(self=@n3)
    ) {blendType=Linear})
  ]
 }
 Open_Into {
  path=AnimGraph.AnimGraphNode_StateMachine_1.Locomotion SM.AnimStateNode_1.Open_Into
  nodes [
   n1 = Break Fort Anim Input Skydiving(FortAnimInput_Skydiving=Get SkydivingState()) -> (
    ParaGilderRootOffset=~
    ParaGliderRootRotationOverride=~
    bUseParaGlideRootModifier=~
    bIsSkydivingFromLaunchPad=~
    bIsSkydivingFromBus=~
    bIsInVortex=~
    bIsDiving=~
    bIsParachuteOpen=~
    bIsSkydiving=~
    bIsParachuteLeaning=~
    bIsSkydiveLeaning=~
    bIsLeaning=~
    bIsSkydiveDiveMode=~
    bPlaySkydiveDrift=~
    bSkydiveDriftDelayActive=~
    bSkydiveDriftAnimAllowed=~
    bIsGliderRight=~
    bIsGliderCenter=~
    bIsGliderLeft=~
    bIsGliderForward=~
    bIsGliderBack=~
    bHasDeployedAndIsStandingUpright=~
    LocalAccelForward=~
    LocalAccelRight=~
    LocalVelocityRight=~
    SkydiveAimPitch=~
    SkydiveAimYaw=~
    DeployChuteAnimRate=~
    SkydiveDriftAnimRate=~
    SkydiveDriftAnimRateCurrent=~
    SkydiveFidgetAnimRate=~
    SkydiveFidgetAnimRateCurrent=~
    SkydiveAdditiveAlpha=~
    SkydiveDriftDelay=~
   )
   n2 = Sequence Player 'SizeTribe_Into_FrontEnd_Anim_GLIDER'() {bLoopAnimation=false, sequence=&SizeTribe_Into_FrontEnd_Anim_GLIDER}
   n3 = Get OwnerGlider()
   Output Animation Pose(Result=Blend Poses by bool(
     BlendPose_0=@n2
     BlendPose_1=Blendspace Player 'BS_SizeTribe_Into_GLIDER'(X=@n1.SkydiveAimYaw, Y=@n1.SkydiveAimPitch) {bLoop=false, blendSpace=&/BS_SizeTribe_Into_GLIDER}
     bActiveValue=Get IsCosmeticPreview(self=@n3)
    ) {blendType=Linear})
  ]
 }
 Closed {
  path=AnimGraph.AnimGraphNode_StateMachine_1.Locomotion SM.AnimStateNode_2.Closed
  nodes [
   n1 = Get OwnerGlider()
   Output Animation Pose(Result=Blend Poses by bool(
     BlendPose_0=Sequence Player 'SizeTribe_Closed_GLIDER'() {sequence=&SizeTribe_Closed_GLIDER}
     BlendPose_1=Sequence Player 'SizeTribe_Closed_GLIDER'() {sequence=&SizeTribe_Closed_GLIDER}
     bActiveValue=Get IsCosmeticPreview(self=@n1)
    ) {blendType=Linear})
  ]
 }
 Transition {path=AnimGraph.AnimGraphNode_StateMachine_1.Locomotion SM.AnimStateTransitionNode_0.Transition, nodes [Result()]}
 Transition {
  path=AnimGraph.AnimGraphNode_StateMachine_1.Locomotion SM.AnimStateTransitionNode_1.Transition
  nodes [Result(bCanEnterTransition=OR Boolean(A=Get Is Parachute Open() -> (Is Parachute Open:bool), B=Get bForceOpen()))]
 }
 Transition {
  path=AnimGraph.AnimGraphNode_StateMachine_1.Locomotion SM.AnimStateTransitionNode_2.Transition
  nodes [
   n1 = Get bForceOpen()
   n2 = Get Is Parachute Open()
   Result(bCanEnterTransition=AND Boolean(A=NOT Boolean(A=@n2), B=NOT Boolean(A=@n1)))
  ]
 }
 Transition {
  path=AnimGraph.AnimGraphNode_StateMachine_1.Locomotion SM.AnimStateTransitionNode_3.Transition
  nodes [
   n1 = Get bForceOpen()
   n2 = Get Is Parachute Open()
   Result(bCanEnterTransition=AND Boolean(A=NOT Boolean(A=@n2), B=NOT Boolean(A=@n1)))
  ]
 }
 Lean Transitions {
  entry=Lean_Zero
  path=AnimGraph.AnimGraphNode_StateMachine_2.Lean Transitions
  ids=id
  nodes [
   AnimStateNode_0 = Lean_Zero() {
    is=title AnimStateNode
    stateEntered {guid=#1}
    stateFullyBlended {guid=#2}
    stateLeft {guid=#3}
   }
   AnimStateNode_2 = Lean_Out() {is=title AnimStateNode}
   AnimStateNode_5 = Lean_Into() {
    is=title AnimStateNode
    stateEntered {guid=#4}
    stateFullyBlended {guid=#5}
    stateLeft {guid=#6}
   }
   AnimStateTransitionNode_0 = AnimStateTransitionNode()
   AnimStateTransitionNode_1 = AnimStateTransitionNode() {bAllowInertializationForSelfTransitions=true}
   AnimStateTransitionNode_11 = AnimStateTransitionNode()
   AnimStateTransitionNode_18 = AnimStateTransitionNode() {bAutomaticRuleBasedOnSequencePlayerInState=true, blendMode=Linear, priorityOrder=2}
  ]
  links [
   {
    dst=Lean_Out
    rule=AnimGraph.AnimGraphNode_StateMachine_2.Lean Transitions.AnimStateTransitionNode_11.Transition
    src=Lean_Into
    via=AnimStateTransitionNode_11
   }
   {
    dst=Lean_Into
    rule=AnimGraph.AnimGraphNode_StateMachine_2.Lean Transitions.AnimStateTransitionNode_1.Transition
    src=Lean_Out
    via=AnimStateTransitionNode_1
   }
   {
    dst=Lean_Zero
    rule=AnimGraph.AnimGraphNode_StateMachine_2.Lean Transitions.AnimStateTransitionNode_18.Transition
    src=Lean_Out
    via=AnimStateTransitionNode_18
   }
   {
    dst=Lean_Into
    rule=AnimGraph.AnimGraphNode_StateMachine_2.Lean Transitions.AnimStateTransitionNode_0.Transition
    src=Lean_Zero
    via=AnimStateTransitionNode_0
   }
  ]
 }
 Lean_Zero {
  path=AnimGraph.AnimGraphNode_StateMachine_2.Lean Transitions.AnimStateNode_0.Lean_Zero
  nodes [Output Animation Pose(Result="Sequence Player (additive) 'RampCanal_Lean_Center_Pose_GLIDER'"() {sequence=/BRCosmetics/Animation/Game/MainPlayer/Skydive/ParaGlide/RampCanal/RampCanal_Lean_Center_Pose_GLIDER})]
 }
 Lean_Out {
  path=AnimGraph.AnimGraphNode_StateMachine_2.Lean Transitions.AnimStateNode_2.Lean_Out
  nodes [
   n1 = "Sequence Player (additive) 'SizeTribe_Lean_Back_Out_Anim_GLIDER'"() {bLoopAnimation=false, sequence=&/SizeTribe_Lean_Back_Out_Anim_GLIDER}
   n2 = "Sequence Player (additive) 'SizeTribe_Lean_Left_Out_Anim_GLIDER'"() {bLoopAnimation=false, sequence=&/SizeTribe_Lean_Left_Out_Anim_GLIDER}
   n3 = "Sequence Player (additive) 'SizeTribe_Lean_Right_Out_Anim_GLIDER'"() {bLoopAnimation=false, sequence=&/SizeTribe_Lean_Right_Out_Anim_GLIDER}
   n4 = "Sequence Player (additive) 'SizeTribe_Lean_Forward_Out_Anim_GLIDER'"() {bLoopAnimation=false, sequence=&/SizeTribe_Lean_Forward_Out_Anim_GLIDER}
   n5 = "Blend Poses (ESkydivingDirection)"!(
    BlendPose_0:Pose Link Structure="Sequence Player (additive) 'SizeTribe_Lean_Center_Pose_GLIDER'"() {bLoopAnimation=false, sequence=&/SizeTribe_Lean_Center_Pose_GLIDER}
    BlendPose_1:Pose Link Structure=@n4
    BlendPose_2:Pose Link Structure=@n1
    BlendPose_3:Pose Link Structure=@n2
    BlendPose_4:Pose Link Structure=@n3
    BlendTime_0:float=1
    BlendTime_1:float=1
    BlendTime_2:float=1
    BlendTime_3:float=1
    BlendTime_4:float=1
    ActiveEnumValue:Byte=Get Direction Last() -> (Direction Last:ESkydivingDirection Enum=Center)
   ) -> (Pose:Pose Link Structure) {
    is=title AnimGraphNode_BlendListByEnum
    blendType=Linear
    blendPose [
     {linkId=-1, sourceLinkId=-1}
     {linkId=-1, sourceLinkId=-1}
     {linkId=-1, sourceLinkId=-1}
     {linkId=-1, sourceLinkId=-1}
     {linkId=-1, sourceLinkId=-1}
    ]
    blendTime [0.1,0.1,0.1,0.1,0.1]
   }
   Output Animation Pose(Result=@n5)
  ]
 }
 Lean_Into {
  path=AnimGraph.AnimGraphNode_StateMachine_2.Lean Transitions.AnimStateNode_5.Lean_Into
  nodes [
   AnimGraphNode_StateMachine_0 = Lean_Into() {is=title AnimGraphNode_StateMachine}
   Output Animation Pose(Result=@AnimGraphNode_StateMachine_0)
  ]
 }
 Lean_Into {
  entry=Zero
  path=AnimGraph.AnimGraphNode_StateMachine_2.Lean Transitions.AnimStateNode_5.Lean_Into.AnimGraphNode_StateMachine_0.Lean_Into
  ids=id
  nodes [
   AnimStateNode_0 = Zero() {is=title AnimStateNode}
   AnimStateNode_1 = Lean_Forward_Into() {is=title AnimStateNode}
   AnimStateNode_2 = Lean_Back_Into() {is=title AnimStateNode}
   AnimStateNode_3 = Lean_Left_Into() {is=title AnimStateNode}
   AnimStateNode_4 = Lean_Right_Into() {is=title AnimStateNode}
   AnimStateTransitionNode_0 = AnimStateTransitionNode()
   AnimStateTransitionNode_1 = AnimStateTransitionNode()
   AnimStateTransitionNode_10 = AnimStateTransitionNode() {crossfadeDuration=0.5}
   AnimStateTransitionNode_11 = AnimStateTransitionNode() {crossfadeDuration=0.5}
   AnimStateTransitionNode_12 = AnimStateTransitionNode() {crossfadeDuration=0.5}
   AnimStateTransitionNode_13 = AnimStateTransitionNode() {crossfadeDuration=0.5}
   AnimStateTransitionNode_14 = AnimStateTransitionNode() {crossfadeDuration=0.5}
   AnimStateTransitionNode_15 = AnimStateTransitionNode() {crossfadeDuration=0.5}
   AnimStateTransitionNode_2 = AnimStateTransitionNode()
   AnimStateTransitionNode_3 = AnimStateTransitionNode()
   AnimStateTransitionNode_4 = AnimStateTransitionNode() {crossfadeDuration=0.5}
   AnimStateTransitionNode_5 = AnimStateTransitionNode() {crossfadeDuration=0.5}
   AnimStateTransitionNode_6 = AnimStateTransitionNode() {crossfadeDuration=0.5}
   AnimStateTransitionNode_7 = AnimStateTransitionNode() {crossfadeDuration=0.5}
   AnimStateTransitionNode_8 = AnimStateTransitionNode() {crossfadeDuration=0.5}
   AnimStateTransitionNode_9 = AnimStateTransitionNode() {crossfadeDuration=0.5}
  ]
  links [
   {
    dst=Lean_Forward_Into
    rule=AnimGraph.AnimGraphNode_StateMachine_2.Lean Transitions.AnimStateNode_5.Lean_Into.AnimGraphNode_StateMachine_0.Lean_Into.AnimStateTransitionNode_6.AnimationTransitionGraph_0
    src=Lean_Back_Into
    via=AnimStateTransitionNode_4
   }
   {
    dst=Lean_Left_Into
    rule=AnimGraph.AnimGraphNode_StateMachine_2.Lean Transitions.AnimStateNode_5.Lean_Into.AnimGraphNode_StateMachine_0.Lean_Into.AnimStateTransitionNode_13.AnimationTransitionGraph_0
    src=Lean_Back_Into
    via=AnimStateTransitionNode_15
   }
   {
    dst=Lean_Right_Into
    rule=AnimGraph.AnimGraphNode_StateMachine_2.Lean Transitions.AnimStateNode_5.Lean_Into.AnimGraphNode_StateMachine_0.Lean_Into.AnimStateTransitionNode_7.AnimationTransitionGraph_0
    src=Lean_Back_Into
    via=AnimStateTransitionNode_8
   }
   {
    dst=Lean_Back_Into
    rule=AnimGraph.AnimGraphNode_StateMachine_2.Lean Transitions.AnimStateNode_5.Lean_Into.AnimGraphNode_StateMachine_0.Lean_Into.AnimStateTransitionNode_12.AnimationTransitionGraph_0
    src=Lean_Forward_Into
    via=AnimStateTransitionNode_11
   }
   {
    dst=Lean_Left_Into
    rule=AnimGraph.AnimGraphNode_StateMachine_2.Lean Transitions.AnimStateNode_5.Lean_Into.AnimGraphNode_StateMachine_0.Lean_Into.AnimStateTransitionNode_13.AnimationTransitionGraph_0
    src=Lean_Forward_Into
    via=AnimStateTransitionNode_13
   }
   {
    dst=Lean_Right_Into
    rule=AnimGraph.AnimGraphNode_StateMachine_2.Lean Transitions.AnimStateNode_5.Lean_Into.AnimGraphNode_StateMachine_0.Lean_Into.AnimStateTransitionNode_7.AnimationTransitionGraph_0
    src=Lean_Forward_Into
    via=AnimStateTransitionNode_7
   }
   {
    dst=Lean_Back_Into
    rule=AnimGraph.AnimGraphNode_StateMachine_2.Lean Transitions.AnimStateNode_5.Lean_Into.AnimGraphNode_StateMachine_0.Lean_Into.AnimStateTransitionNode_12.AnimationTransitionGraph_0
    src=Lean_Left_Into
    via=AnimStateTransitionNode_10
   }
   {
    dst=Lean_Forward_Into
    rule=AnimGraph.AnimGraphNode_StateMachine_2.Lean Transitions.AnimStateNode_5.Lean_Into.AnimGraphNode_StateMachine_0.Lean_Into.AnimStateTransitionNode_6.AnimationTransitionGraph_0
    src=Lean_Left_Into
    via=AnimStateTransitionNode_6
   }
   {
    dst=Lean_Right_Into
    rule=AnimGraph.AnimGraphNode_StateMachine_2.Lean Transitions.AnimStateNode_5.Lean_Into.AnimGraphNode_StateMachine_0.Lean_Into.AnimStateTransitionNode_7.AnimationTransitionGraph_0
    src=Lean_Left_Into
    via=AnimStateTransitionNode_9
   }
   {
    dst=Lean_Back_Into
    rule=AnimGraph.AnimGraphNode_StateMachine_2.Lean Transitions.AnimStateNode_5.Lean_Into.AnimGraphNode_StateMachine_0.Lean_Into.AnimStateTransitionNode_12.AnimationTransitionGraph_0
    src=Lean_Right_Into
    via=AnimStateTransitionNode_12
   }
   {
    dst=Lean_Forward_Into
    rule=AnimGraph.AnimGraphNode_StateMachine_2.Lean Transitions.AnimStateNode_5.Lean_Into.AnimGraphNode_StateMachine_0.Lean_Into.AnimStateTransitionNode_6.AnimationTransitionGraph_0
    src=Lean_Right_Into
    via=AnimStateTransitionNode_5
   }
   {
    dst=Lean_Left_Into
    rule=AnimGraph.AnimGraphNode_StateMachine_2.Lean Transitions.AnimStateNode_5.Lean_Into.AnimGraphNode_StateMachine_0.Lean_Into.AnimStateTransitionNode_13.AnimationTransitionGraph_0
    src=Lean_Right_Into
    via=AnimStateTransitionNode_14
   }
   {
    dst=Lean_Back_Into
    rule=AnimGraph.AnimGraphNode_StateMachine_2.Lean Transitions.AnimStateNode_5.Lean_Into.AnimGraphNode_StateMachine_0.Lean_Into.AnimStateTransitionNode_12.AnimationTransitionGraph_0
    src=Zero
    via=AnimStateTransitionNode_1
   }
   {
    dst=Lean_Forward_Into
    rule=AnimGraph.AnimGraphNode_StateMachine_2.Lean Transitions.AnimStateNode_5.Lean_Into.AnimGraphNode_StateMachine_0.Lean_Into.AnimStateTransitionNode_6.AnimationTransitionGraph_0
    src=Zero
    via=AnimStateTransitionNode_0
   }
   {
    dst=Lean_Left_Into
    rule=AnimGraph.AnimGraphNode_StateMachine_2.Lean Transitions.AnimStateNode_5.Lean_Into.AnimGraphNode_StateMachine_0.Lean_Into.AnimStateTransitionNode_13.AnimationTransitionGraph_0
    src=Zero
    via=AnimStateTransitionNode_2
   }
   {
    dst=Lean_Right_Into
    rule=AnimGraph.AnimGraphNode_StateMachine_2.Lean Transitions.AnimStateNode_5.Lean_Into.AnimGraphNode_StateMachine_0.Lean_Into.AnimStateTransitionNode_7.AnimationTransitionGraph_0
    src=Zero
    via=AnimStateTransitionNode_3
   }
  ]
 }
 Zero {
  path=AnimGraph.AnimGraphNode_StateMachine_2.Lean Transitions.AnimStateNode_5.Lean_Into.AnimGraphNode_StateMachine_0.Lean_Into.AnimStateNode_0.Zero
  nodes [Output Animation Pose(Result=Additive Identity Pose())]
 }
 Lean_Forward_Into {
  path=AnimGraph.AnimGraphNode_StateMachine_2.Lean Transitions.AnimStateNode_5.Lean_Into.AnimGraphNode_StateMachine_0.Lean_Into.AnimStateNode_1.Lean_Forward_Into
  nodes [
   n1 = "Sequence Player (additive) 'SizeTribe_Lean_Forward_Into_Anim_GLIDER'"() {bLoopAnimation=false, sequence=&/SizeTribe_Lean_Forward_Into_Anim_GLIDER}
   Output Animation Pose(Result=@n1)
  ]
 }
 Lean_Back_Into {
  path=AnimGraph.AnimGraphNode_StateMachine_2.Lean Transitions.AnimStateNode_5.Lean_Into.AnimGraphNode_StateMachine_0.Lean_Into.AnimStateNode_2.Lean_Back_Into
  nodes [
   n1 = "Sequence Player (additive) 'SizeTribe_Lean_Back_Into_Anim_GLIDER'"() {bLoopAnimation=false, sequence=&/SizeTribe_Lean_Back_Into_Anim_GLIDER}
   Output Animation Pose(Result=@n1)
  ]
 }
 Lean_Left_Into {
  path=AnimGraph.AnimGraphNode_StateMachine_2.Lean Transitions.AnimStateNode_5.Lean_Into.AnimGraphNode_StateMachine_0.Lean_Into.AnimStateNode_3.Lean_Left_Into
  nodes [
   n1 = "Sequence Player (additive) 'SizeTribe_Lean_Left_Into_Anim_GLIDER'"() {bLoopAnimation=false, sequence=&/SizeTribe_Lean_Left_Into_Anim_GLIDER}
   Output Animation Pose(Result=@n1)
  ]
 }
 Lean_Right_Into {
  path=AnimGraph.AnimGraphNode_StateMachine_2.Lean Transitions.AnimStateNode_5.Lean_Into.AnimGraphNode_StateMachine_0.Lean_Into.AnimStateNode_4.Lean_Right_Into
  nodes [
   n1 = "Sequence Player (additive) 'SizeTribe_Lean_Right_Into_Anim_GLIDER'"() {bLoopAnimation=false, sequence=&/SizeTribe_Lean_Right_Into_Anim_GLIDER}
   Output Animation Pose(Result=@n1)
  ]
 }
 AnimationTransitionGraph_0 {
  path=AnimGraph.AnimGraphNode_StateMachine_2.Lean Transitions.AnimStateNode_5.Lean_Into.AnimGraphNode_StateMachine_0.Lean_Into.AnimStateTransitionNode_12.AnimationTransitionGraph_0
  nodes [Result(bCanEnterTransition=Get Is Glider Back() -> (Is Glider Back:bool))]
 }
 AnimationTransitionGraph_0 {
  path=AnimGraph.AnimGraphNode_StateMachine_2.Lean Transitions.AnimStateNode_5.Lean_Into.AnimGraphNode_StateMachine_0.Lean_Into.AnimStateTransitionNode_13.AnimationTransitionGraph_0
  nodes [Result(bCanEnterTransition=Get Is Glider Left() -> (Is Glider Left:bool))]
 }
 AnimationTransitionGraph_0 {
  path=AnimGraph.AnimGraphNode_StateMachine_2.Lean Transitions.AnimStateNode_5.Lean_Into.AnimGraphNode_StateMachine_0.Lean_Into.AnimStateTransitionNode_6.AnimationTransitionGraph_0
  nodes [Result(bCanEnterTransition=Get Is Glider Forward() -> (Is Glider Forward:bool))]
 }
 AnimationTransitionGraph_0 {
  path=AnimGraph.AnimGraphNode_StateMachine_2.Lean Transitions.AnimStateNode_5.Lean_Into.AnimGraphNode_StateMachine_0.Lean_Into.AnimStateTransitionNode_7.AnimationTransitionGraph_0
  nodes [Result(bCanEnterTransition=Get Is Glider Right() -> (Is Glider Right:bool))]
 }
 Transition {
  path=AnimGraph.AnimGraphNode_StateMachine_2.Lean Transitions.AnimStateTransitionNode_0.Transition
  nodes [Result(bCanEnterTransition=Get Is Leaning() -> (Is Leaning:bool))]
 }
 Transition {
  path=AnimGraph.AnimGraphNode_StateMachine_2.Lean Transitions.AnimStateTransitionNode_1.Transition
  nodes [Result(bCanEnterTransition=Get bIsAccelerating2D())]
 }
 Transition {
  path=AnimGraph.AnimGraphNode_StateMachine_2.Lean Transitions.AnimStateTransitionNode_11.Transition
  nodes [Result(bCanEnterTransition=NOT Boolean(A=Get bIsAccelerating2D()))]
 }
 Transition {path=AnimGraph.AnimGraphNode_StateMachine_2.Lean Transitions.AnimStateTransitionNode_18.Transition, nodes [Result()]}
 Blink SM {
  entry=Additive Identify
  path=AnimGraph.AnimGraphNode_StateMachine_3.Blink SM
  ids=id
  nodes [
   AnimStateNode_0 = Additive Identify() {is=title AnimStateNode}
   AnimStateNode_1 = Blink() {is=title AnimStateNode}
   AnimStateTransitionNode_0 = AnimStateTransitionNode() {bAllowInertializationForSelfTransitions=true}
   AnimStateTransitionNode_1 = AnimStateTransitionNode() {
    automaticRuleTriggerTime=0
    bAllowInertializationForSelfTransitions=true
    bAutomaticRuleBasedOnSequencePlayerInState=true
    crossfadeDuration=0
    minTimeBeforeReentry=0
   }
  ]
  links [
   {
    dst=Blink
    rule=AnimGraph.AnimGraphNode_StateMachine_3.Blink SM.AnimStateTransitionNode_0.Transition
    src=Additive Identify
    via=AnimStateTransitionNode_0
   }
   {
    dst=Additive Identify
    rule=AnimGraph.AnimGraphNode_StateMachine_3.Blink SM.AnimStateTransitionNode_1.Transition
    src=Blink
    via=AnimStateTransitionNode_1
   }
  ]
 }
 Additive Identify {
  path=AnimGraph.AnimGraphNode_StateMachine_3.Blink SM.AnimStateNode_0.Additive Identify
  nodes [Output Animation Pose(Result=Additive Identity Pose())]
 }
 Blink {
  path=AnimGraph.AnimGraphNode_StateMachine_3.Blink SM.AnimStateNode_1.Blink
  nodes [Output Animation Pose(Result="Sequence Player (additive) 'SizeTribe_Blink_Anim_GLIDER'"() {bLoopAnimation=false, sequence=&/SizeTribe_Blink_Anim_GLIDER})]
 }
 Transition {
  path=AnimGraph.AnimGraphNode_StateMachine_3.Blink SM.AnimStateTransitionNode_0.Transition
  nodes [Result(bCanEnterTransition=Get IsBlinking() -> (IsBlinking:bool))]
 }
 Transition {path=AnimGraph.AnimGraphNode_StateMachine_3.Blink SM.AnimStateTransitionNode_1.Transition, nodes [Result()]}
 EventGraph {nodes [
   n1 = Event BlueprintUpdateAnimation()
   > Set Delta Time X(Delta Time X=@n1.DeltaTimeX)
   > n2 = Sequence!(execute) -> (then_0, then_1, then_2:Exec, then_3:Exec)
   Reroute Node(InputPin=@n2.then_3)
   n3 = ForwardBack_Value_Interp!(
    execute=@n2.then_2
    Time:double=Get Delta Time X()
    Currnet_Value:double=Get AccelRate()
    Idle_Value:double=1
    Idle_InterpSpeed:double=1
    Forward_Value:double=2
    Forward_InterpSpeed:double=2
    Back_Value:double=1
    Back_InterpSpeed:double=2
    LocalAccelForward:double=Get Local Accel Forward() -> (Local Accel Forward:float)
   ) -> (then, New_Value:double)
   > Set AccelRate(AccelRate:double=@n3) -> (Output_Get:double)
   n4 = Sequence!(execute=@n2.then_1) -> (then_0, then_1, then_2, then_3, then_4:Exec, then_5:Exec, then_6:Exec)
   n5 = Get SkeletalMesh!(self:Skinned Mesh Component Object Reference=GetOwningComponent()) -> (SkeletalMesh:Skeletal Mesh Object Reference)
   > n6 = GetComponentSpaceBoneTransform!(execute, SkeletalMesh:Skeletal Mesh Object Reference=@n5, BoneName:Name=joint_wrist_l) -> (then, Valid:bool, ReturnValue:Transform)
   > Branch(Condition=@n6)
   > Set joint_wrist_l(joint_wrist_l:Transform=@n6.ReturnValue) -> (Output_Get:Transform)
   n7 = GetComponentSpaceBoneTransform(execute=@n4.then_1, SkeletalMesh=@n5, BoneName=Wing_l)
   > Branch(Condition=@n7)
   > Set Wing_l(Wing_l:Transform=@n7.ReturnValue) -> (Output_Get:Transform)
   n8 = GetComponentSpaceBoneTransform(execute=@n4.then_2, SkeletalMesh=@n5, BoneName=joint_wrist_r)
   > Branch(Condition=@n8)
   > Set joint_wrist_r(joint_wrist_r:Transform=@n8.ReturnValue) -> (Output_Get:Transform)
   n9 = GetComponentSpaceBoneTransform(execute=@n4.then_3, SkeletalMesh=@n5, BoneName=Wing_r)
   > Branch(Condition=@n9)
   > Set Wing_r(Wing_r:Transform=@n9.ReturnValue) -> (Output_Get:Transform)
   n10 = GetComponentSpaceBoneTransform(execute=@n4.then_4, SkeletalMesh=@n5, BoneName=joint_tails)
   > Branch(Condition=@n10)
   > n11 = Set joint_tails!(execute, joint_tails:Transform=@n10.ReturnValue) -> (
    then
    Output_Get_Location:Vector
    Output_Get_Rotation_Roll:float
    Output_Get_Rotation_Pitch:float
    Output_Get_Rotation_Yaw:float
    Output_Get_Scale:Vector
   )
   > n12 = Branch(Condition=Get IsCosmeticPreview(self=Get OwnerGlider()))
   > Set!(
    execute
    jointTailsRotator_Roll:float=Get SpinRotator() -> (SpinRotator:double)
    jointTailsRotator_Pitch:float=float + float(A=@n11.Output_Get_Rotation_Pitch, B=40)
    jointTailsRotator_Yaw:float=@n11.Output_Get_Rotation_Yaw
   ) -> (then, Output_Get:Rotator)
   Set(
    execute=@n12.else
    jointTailsRotator_Roll=Get SpinRotator()
    jointTailsRotator_Pitch=float + float(A=@n11.Output_Get_Rotation_Pitch, B=40)
    jointTailsRotator_Yaw=@n11.Output_Get_Rotation_Yaw
   )
   n13 = GetComponentSpaceBoneTransform(execute=@n4.then_5, SkeletalMesh=@n5, BoneName=joint_tails_end)
   > Branch(Condition=@n13)
   > n14 = Set joint_tails_end!(execute, joint_tails_end:Transform=@n13.ReturnValue) -> (
    then
    Output_Get_Location:Vector
    Output_Get_Rotation_Roll:float
    Output_Get_Rotation_Pitch:float
    Output_Get_Rotation_Yaw:float
    Output_Get_Scale:Vector
   )
   > n15 = Branch(Condition=Get IsCosmeticPreview(self=Get OwnerGlider()))
   > Set!(
    execute
    jointTailsEndRotator_Roll:float=Get SpinRotator()
    jointTailsEndRotator_Pitch:float=float + float(A=@n14.Output_Get_Rotation_Pitch, B=30)
    jointTailsEndRotator_Yaw:float=@n14.Output_Get_Rotation_Yaw
   ) -> (then, Output_Get)
   Set(
    execute=@n15.else
    jointTailsEndRotator_Roll=Get SpinRotator()
    jointTailsEndRotator_Pitch=float + float(A=@n14.Output_Get_Rotation_Pitch, B=30)
    jointTailsEndRotator_Yaw=@n14.Output_Get_Rotation_Yaw
   )
   n16 = GetComponentSpaceBoneTransform(execute=@n4.then_6, SkeletalMesh=@n5, BoneName=joint_tails_end)
   > Branch(Condition=@n16)
   > n17 = Set joint_tails_end!(execute, joint_tails_end=@n16.ReturnValue) -> (
    then
    Output_Get_Location_X:double
    Output_Get_Location_Y:double
    Output_Get_Location_Z:double
    Output_Get_Rotation:Rotator
    Output_Get_Scale
   )
   > Set!(
    execute
    joinTailsEndLocation_X:double=@n17
    joinTailsEndLocation_Y:double=float + float(A=@n17.Output_Get_Location_Y)
    joinTailsEndLocation_Z:double=float + float(A=@n17.Output_Get_Location_Z)
   ) -> (then, Output_Get:Vector)
   GetUsefulVariables(execute=@n2.then_0)
   > SetSpinVariable()
   AnimNotify_TurnONEngines() {is=title K2Node_Event}
   > SetGliderFullyDeployed!(execute, self:Fort Glider Anim Instance Object Reference, bIsFullyDeployed:bool=true) -> (then)
   n18 = AnimNotify_ClosedEntered() {is=title K2Node_Event}
   n19 = GetOwningComponent()
   HideBoneByName!(
    execute=[@n18.then, @n21.then, @n23.then]
    self:Skinned Mesh Component Object Reference=@n19
    BoneName:Name=Glider_Base
    PhysBodyOption:EPhysBodyOp Enum=PBO_None
   ) -> (then)
   n20 = AnimNotify_DeployEntered() {is=title K2Node_Event}
   Delay(execute=[@n20.then, @n22.then], Duration=0.1)
   > UnHideBoneByName!(execute, self:Skinned Mesh Component Object Reference=@n19, BoneName:Name=Glider_Base) -> (then)
   n21 = Event BlueprintInitializeAnimation()
   n22 = AnimNotify_UnhideGlider() {is=title K2Node_Event}
   n23 = AnimNotify_Closed() {is=title K2Node_Event}
   AnimNotify_CanBlink() {is=title K2Node_Event}
   > Set CanBlink(CanBlink:bool=true) -> (Output_Get:bool)
   AnimNotify_ReBlink() {is=title K2Node_Event}
   > Delay(Duration=RandomFloatInRange(Min=2, Max=5))
   > Set IsBlinking(IsBlinking:bool=true) -> (Output_Get:bool)
   > Delay(Duration=1)
   > Set IsBlinking(IsBlinking=false)
   Branch(Condition=Get IsCosmeticPreview(self=Get OwnerGlider()))
   > n24 = Branch(Condition=Get CanBlink() -> (CanBlink:bool))
   > Do Once(Reset=@n24.else)
   > Delay(Duration=2)
   > Set IsBlinking(IsBlinking=true)
   > Delay(Duration=1)
   > Set IsBlinking(IsBlinking=false)
   n25 = Get OwnerGlider()
   Set Actor Hidden In Game!(execute, self:Actor Object Reference=@n25, bNewHidden:bool=true) -> (then) {is=title K2Node_CallFunction}
   Set Actor Hidden In Game(self=@n25, bNewHidden=true)
  ]}
 GetUsefulVariables {nodes [
   GetUsefulVariables() {is=title K2Node_FunctionEntry}
   > n1 = Sequence!(execute) -> (then_0, then_1)
   n2 = Break Fort Anim Input Skydiving(FortAnimInput_Skydiving=Get SkydivingState()) -> (
    ParaGilderRootOffset=~
    ParaGliderRootRotationOverride=~
    bUseParaGlideRootModifier=~
    bIsSkydivingFromLaunchPad=~
    bIsSkydivingFromBus=~
    bIsInVortex=~
    bIsDiving=~
    bIsParachuteOpen=~
    bIsSkydiving=~
    bIsParachuteLeaning=~
    bIsSkydiveLeaning=~
    bIsLeaning=~
    bIsSkydiveDiveMode=~
    bPlaySkydiveDrift=~
    bSkydiveDriftDelayActive=~
    bSkydiveDriftAnimAllowed=~
    bIsGliderRight=~
    bIsGliderCenter=~
    bIsGliderLeft=~
    bIsGliderForward=~
    bIsGliderBack=~
    bHasDeployedAndIsStandingUpright=~
    LocalAccelForward=~
    LocalAccelRight=~
    LocalVelocityRight=~
    SkydiveAimPitch=~
    SkydiveAimYaw=~
    DeployChuteAnimRate=~
    SkydiveDriftAnimRate=~
    SkydiveDriftAnimRateCurrent=~
    SkydiveFidgetAnimRate=~
    SkydiveFidgetAnimRateCurrent=~
    SkydiveAdditiveAlpha=~
    SkydiveDriftDelay=~
   )
   > Set Is Parachute Open(Is Parachute Open:bool=@n2.bIsParachuteOpen) -> (Output_Get:bool)
   > Set Local Accel Forward(Local Accel Forward:float=@n2.LocalAccelForward) -> (Output_Get:float)
   > Set Local Accel Right(Local Accel Right:float=@n2.LocalAccelRight) -> (Output_Get:float)
   > Set Skydive Aim Pitch(Skydive Aim Pitch:float=@n2.SkydiveAimPitch) -> (Output_Get:float)
   > Set Skydive Aim Yaw(Skydive Aim Yaw:float=@n2.SkydiveAimYaw) -> (Output_Get:float)
   > Set Is Leaning(Is Leaning:bool=@n2.bIsLeaning) -> (Output_Get:bool)
   > Set Direction Last(Direction Last:ESkydivingDirection Enum=@n2.DirectionLast) -> (Output_Get:ESkydivingDirection Enum=Center)
   n3 = Break Fort Anim Input Skydiving!(FortAnimInput_Skydiving=Get SkydivingState()) -> (
    bIsGliderRight=~
    bIsGliderCenter=~
    bIsGliderLeft=~
    bIsGliderForward=~
    bIsGliderBack=~
    bHasDeployedAndIsStandingUpright=~
    LocalAccelDir
    PoseAdditive
   )
   Set Is Glider Right(execute=@n1.then_1, Is Glider Right:bool=@n3) -> (Output_Get:bool)
   > Set Is Glider Center(Is Glider Center:bool=@n3.bIsGliderCenter) -> (Output_Get:bool)
   > Set Is Glider Left(Is Glider Left:bool=@n3.bIsGliderLeft) -> (Output_Get:bool)
   > Set Is Glider Forward(Is Glider Forward:bool=@n3.bIsGliderForward) -> (Output_Get:bool)
   > Set Is Glider Back(Is Glider Back:bool=@n3.bIsGliderBack) -> (Output_Get:bool)
  ]}
 SetSpinVariable {nodes [
   SetSpinVariable() {is=title K2Node_FunctionEntry}
   > Set LocalSpinValue(LocalSpinValue:double=float + float(A=Get SpinRotator(), B=50)) -> (Output_Get:double)
   > Set SpinRotator(SpinRotator:double=FInterpTo(Current=Get SpinRotator(), Target=Get LocalSpinValue() -> (LocalSpinValue:double), DeltaTime=Get Delta Time X(), InterpSpeed=30)) -> (Output_Get:double)
  ]}
]
variables [
 AccelRate {type=number}
 CanBlink {type=boolean}
 Delta Time X {type=number}
 Direction Last {type=ESkydivingDirection, default=Center}
 Is Glider Back {type=boolean}
 Is Glider Center {type=boolean}
 Is Glider Forward {type=boolean}
 Is Glider Left {type=boolean}
 Is Glider Right {type=boolean}
 Is Leaning {type=boolean}
 Is Parachute Open {type=boolean}
 IsBlinking {type=boolean}
 joint_tails {type=Transform, default {location=<0 0 0>, rotation=<rot 0 0 0>, scale=<1 1 1>}}
 joint_tails_end {type=Transform, default {location=<0 0 0>, rotation=<rot 0 0 0>, scale=<1 1 1>}}
 joint_wrist_l {type=Transform, default {location=<0 0 0>, rotation=<rot 0 0 0>, scale=<1 1 1>}}
 joint_wrist_r {type=Transform, default {location=<0 0 0>, rotation=<rot 0 0 0>, scale=<1 1 1>}}
 joinTailsEndLocation {type=Vector}
 jointTailsEndRotator {type=Rotator}
 jointTailsRotator {type=Rotator}
 Local Accel Forward {type=number}
 Local Accel Right {type=number}
 Skydive Aim Pitch {type=number}
 Skydive Aim Yaw {type=number}
 SpinRotator {type=number}
 Wing_l {type=Transform, default {location=<0 0 0>, rotation=<rot 0 0 0>, scale=<1 1 1>}}
 Wing_r {type=Transform, default {location=<0 0 0>, rotation=<rot 0 0 0>, scale=<1 1 1>}}
]
